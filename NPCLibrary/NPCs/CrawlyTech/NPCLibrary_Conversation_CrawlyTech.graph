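# Conversation database for the CrawlyTech NPC, Doctor Theodore Flyesbee, providing the greeting and top-level routing when a player talks to the character.
# The entry node feeds one Speech node with a default line plus contextual variants for first meetings, conversation counts and specific player outfits.
# First-meeting and outfit lines carry named priority curve rows.
# The speech branches through link nodes to the shared NPC quests and basic services entries, plus CrawlyTech's own dialogue entry.
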ConversationDatabase /NPCLibrary/NPCs/CrawlyTech/NPCLibrary_Conversation_CrawlyTech
entry_points [{tag=Conversation.Entry.CrawlyTech, destinations [#1]}]
nodes [
 ConversationEntryPointNode_7 {
  id=#1
  class=ConversationEntryPointNode
  outputs [#2]
  properties {entryTag {tagName=Conversation.Entry.CrawlyTech}}
 }
 Speech_C_4 {
  id=#2
  class=Speech_C
  outputs [#3, #4, #5]
  properties {
   nodeName=""
   generalConfig {
    defaultMessage="Oh, hi! I was just rubbing my weird little hands together for no reason."
    contextualMessages [
     {
      message="Doctor Theodore Flyesbee, buzzing to be of service!"
      contextRequirements [{requirement {class=FortControllerRequirement_IsFirstEverConversationWithNPC}}]
      priority {curve {rowName=Tandem.Speech.Priority.FirstEverInteraction}}
     }
     {
      message="I'm an expert in parasitic infestations. And this one... is bad!"
      contextRequirements [{requirement {
         class=FortControllerRequirement_CheckNPCMatchConversationCount
         properties {nPCUniqueIdTag {tagName=AISpawnerData.Type.Tandem.CrawlyTech}, requiredConversationCountConfig {requirement=GreaterThan}}
        }}]
      priority {curve {curveTable=null, rowName=None}}
     }
     {
      message="Oh, hi! I was just rubbing my weird little hands together for no reason."
      contextRequirements [{requirement {
         class=FortControllerRequirement_CheckNPCMatchConversationCount
         properties {nPCUniqueIdTag {tagName=AISpawnerData.Type.Tandem.CrawlyTech}, requiredConversationCountConfig {requirement=GreaterThan}}
        }}]
      priority {curve {curveTable=null, rowName=None}}
     }
     {
      message=I'm a PhD - a Perfect Housefly Doctor.
      contextRequirements [{requirement {
         class=FortControllerRequirement_CheckNPCMatchConversationCount
         properties {nPCUniqueIdTag {tagName=AISpawnerData.Type.Tandem.CrawlyTech}, requiredConversationCountConfig {requirement=GreaterThan}}
        }}]
      priority {curve {curveTable=null, rowName=None}}
     }
     {
      message="You're, um... rather strapping for an insectoid..."
      contextRequirements [{requirement {class=FortControllerRequirement_HasCID, properties {allowedCIds [/BRCosmetics/Athena/Items/Cosmetics/Characters/Character_DustyBun]}}}]
      priority {curve {rowName=Tandem.Speech.Priority.OutfitReaction}}
     }
     {
      message=You. Are. Intoxicating.
      contextRequirements [{requirement {
         class=FortControllerRequirement_HasCID
         properties {allowedCIds [/BRCosmetics/Athena/Items/Cosmetics/Characters/CID_A_191_Athena_Commando_M_AlienFlora]}
        }}]
      priority {curve {rowName=Tandem.Speech.Priority.OutfitReaction}}
     }
    ]
   }
  }
 }
 ConversationLinkNode_2 {
  id=#5
  class=ConversationLinkNode
  properties {remoteEntryTag {tagName=Conversation.Entry.CrawlyTech.Dialogue}}
 }
 ConversationLinkNode_0 {id=#4, class=ConversationLinkNode, properties {remoteEntryTag {tagName=Conversation.Entry.BasicServices}}}
 ConversationLinkNode_1 {id=#3, class=ConversationLinkNode, properties {remoteEntryTag {tagName=Conversation.Entry.NPCLibrary.Quests}}}
]
