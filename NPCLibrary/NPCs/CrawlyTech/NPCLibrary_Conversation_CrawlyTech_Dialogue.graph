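# Conversation database for the CrawlyTech NPC's dialogue, giving flavor and story talk about the vanished bugs, the Onimorphs and the character's hobbies.
# A single tagged entry point leads to an "Ask about..." speech node that branches into "The Bugs are gone!", a BACK node and "Hobbies".
# Both topics play as linear chains of speech lines that converge on one shared ReturnToStart node, also labeled BACK.
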ConversationDatabase /NPCLibrary/NPCs/CrawlyTech/NPCLibrary_Conversation_CrawlyTech_Dialogue
entry_points [{tag=Conversation.Entry.CrawlyTech.Dialogue, destinations [#1]}]
nodes [
 ConversationEntryPointNode_0 {
  id=#1
  class=ConversationEntryPointNode
  outputs [#2]
  properties {entryTag {tagName=Conversation.Entry.CrawlyTech.Dialogue}}
 }
 Speech_C_0 {
  id=#2
  class=Speech_C
  outputs [#3, #4, #5]
  properties {
   nodeName=""
   softTaskIcon=&T_UI_NPC_question
   generalConfig {defaultMessage="I'm no fly-by-night expert, I assure you!"}
  }
  sub_nodes [ConversationChoiceNode_5 {class=ConversationChoiceNode, properties {defaultChoiceDisplayText=Ask about...}}]
 }
 Speech_C_2 {
  id=#5
  class=Speech_C
  outputs [#6]
  properties {
   nodeName=""
   softTaskIcon=&T_UI_NPC_question
   generalConfig {defaultMessage="When I'm not studying, I enjoy buzzing in the ears of strangers..."}
  }
  sub_nodes [ConversationChoiceNode_3 {class=ConversationChoiceNode, properties {defaultChoiceDisplayText=Hobbies}}]
 }
 Speech_C_3 {
  id=#6
  class=Speech_C
  outputs [#7]
  properties {
   nodeName=""
   softTaskIcon=&T_UI_NPC_question
   generalConfig {defaultMessage="...getting stuck between sliding glass doors and their natural partners, the sliding screen door..."}
  }
 }
 Speech_C_4 {
  id=#7
  class=Speech_C
  outputs [#8]
  properties {
   nodeName=""
   softTaskIcon=&T_UI_NPC_question
   generalConfig {defaultMessage=and taking in the delicious stench of... unmentionable organic material.}
  }
 }
 Speech_C_5 {
  id=#8
  class=Speech_C
  outputs [#9]
  properties {
   nodeName=""
   softTaskIcon=&T_UI_NPC_question
   generalConfig {defaultMessage="We all have our vices! Better to embrace them, no?"}
  }
 }
 ReturnToStart_C_23 {
  id=#9
  class=ReturnToStart_C
  properties {softTaskIcon=&T_UI_NPC_Back}
  sub_nodes [ConversationChoiceNode_32 {class=ConversationChoiceNode, properties {defaultChoiceDisplayText=BACK}}]
 }
 Back_C_1 {
  id=#4
  class=Back_C
  sub_nodes [ConversationChoiceNode_2 {class=ConversationChoiceNode, properties {defaultChoiceDisplayText=BACK}}]
 }
 Speech_C_7 {
  id=#3
  class=Speech_C
  outputs [#10]
  properties {
   nodeName=""
   softTaskIcon=&T_UI_NPC_question
   generalConfig {defaultMessage="Great gnat above! What do you mean, GONE?!"}
  }
  sub_nodes [ConversationChoiceNode_0 {class=ConversationChoiceNode, properties {defaultChoiceDisplayText="The Bugs are gone!"}}]
 }
 Speech_C_15 {
  id=#10
  class=Speech_C
  outputs [#11]
  properties {
   nodeName=""
   softTaskIcon=&T_UI_NPC_question
   generalConfig {defaultMessage="The Onimorphs were WELL on their way to devouring this reality, I assure you."}
  }
 }
 Speech_C_14 {
  id=#11
  class=Speech_C
  outputs [#12]
  properties {
   nodeName=""
   softTaskIcon=&T_UI_NPC_question
   generalConfig {defaultMessage="These little parisitic profiteroles don't simply up and leave for no reason!"}
  }
 }
 Speech_C_1 {
  id=#12
  class=Speech_C
  outputs [#9]
  properties {
   nodeName=""
   softTaskIcon=&T_UI_NPC_question
   generalConfig {defaultMessage="Unless... Unless something further up the food chain is on its way... Hmm. I should go!"}
  }
 }
]
